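# Sound Cue for large quadruped paw footsteps at a walk without nail clicks, part of the shared CosmeticCompanions foley audio.
# A single Random node at the root picks among twelve numbered wave variations with equal weights.
# The graph has no modulation, mixing or attenuation nodes, so variety comes only from the choice of sample.
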SoundCue /CosmeticCompanions/Assets/Audio_Shared/Foley/Quadraped/Paws_NoNails/Wavs/Mimosa_Quadruped_Paws_Lrg_Walk_NoNails
root=Random_0
nodes [
 {
  id=Random_0
  children [
   WavePlayer_0
   WavePlayer_1
   WavePlayer_2
   WavePlayer_3
   WavePlayer_4
   WavePlayer_5
   WavePlayer_6
   WavePlayer_7
   WavePlayer_8
   WavePlayer_9
   WavePlayer_10
   WavePlayer_11
  ]
  properties {weights [1,1,1,1,1,1,1,1,1,1,1,1]}
 }
 {
  id=WavePlayer_0
  properties {soundWaveAssetPtr=&/Mimosa_Quadruped_Paws_Lrg_Walk_NoNails_01}
 }
 {
  id=WavePlayer_1
  properties {soundWaveAssetPtr=&/Mimosa_Quadruped_Paws_Lrg_Walk_NoNails_02}
 }
 {
  id=WavePlayer_10
  properties {soundWaveAssetPtr=&/Mimosa_Quadruped_Paws_Lrg_Walk_NoNails_11}
 }
 {
  id=WavePlayer_11
  properties {soundWaveAssetPtr=&/Mimosa_Quadruped_Paws_Lrg_Walk_NoNails_12}
 }
 {
  id=WavePlayer_2
  properties {soundWaveAssetPtr=&/Mimosa_Quadruped_Paws_Lrg_Walk_NoNails_03}
 }
 {
  id=WavePlayer_3
  properties {soundWaveAssetPtr=&/Mimosa_Quadruped_Paws_Lrg_Walk_NoNails_04}
 }
 {
  id=WavePlayer_4
  properties {soundWaveAssetPtr=&/Mimosa_Quadruped_Paws_Lrg_Walk_NoNails_05}
 }
 {
  id=WavePlayer_5
  properties {soundWaveAssetPtr=&/Mimosa_Quadruped_Paws_Lrg_Walk_NoNails_06}
 }
 {
  id=WavePlayer_6
  properties {soundWaveAssetPtr=&/Mimosa_Quadruped_Paws_Lrg_Walk_NoNails_07}
 }
 {
  id=WavePlayer_7
  properties {soundWaveAssetPtr=&/Mimosa_Quadruped_Paws_Lrg_Walk_NoNails_08}
 }
 {
  id=WavePlayer_8
  properties {soundWaveAssetPtr=&/Mimosa_Quadruped_Paws_Lrg_Walk_NoNails_09}
 }
 {
  id=WavePlayer_9
  properties {soundWaveAssetPtr=&/Mimosa_Quadruped_Paws_Lrg_Walk_NoNails_10}
 }
]
properties {maxDistance=2097152}
